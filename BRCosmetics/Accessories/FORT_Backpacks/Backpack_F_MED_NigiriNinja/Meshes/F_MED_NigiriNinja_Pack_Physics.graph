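# Physics asset for the NigiriNinja backpack cosmetic, giving its hanging rope chain physics bodies alongside the pack body.
# One large capsule covers the pack, and small capsules cover base_rope and three dyn_rope bones chained in sequence.
# Rope joints have tight twist limits and swing limits that widen toward the tip.
# The base_rope-to-pack joint leaves swing and twist free.
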PhysicsAsset /BRCosmetics/Accessories/FORT_Backpacks/Backpack_F_MED_NigiriNinja/Meshes/F_MED_NigiriNinja_Pack_Physics
bodies {
 base_rope [{shapeType=Capsule, length=2, radius=1, shapeName=base_rope_capsule, center=<2 0 0>, rotation=<rot 0 0 90>}]
 dyn_rope_01 [{shapeType=Capsule, length=2, radius=1, shapeName=dyn_rope_01_capsule, center=<2 0 0>, rotation=<rot 90 0 0>}]
 dyn_rope_02 [{shapeType=Capsule, length=2, radius=1, shapeName=dyn_rope_02_capsule, center=<2 0 0>, rotation=<rot 90 0 0>}]
 dyn_rope_03 [{shapeType=Capsule, length=2, radius=1, shapeName=dyn_rope_03_capsule, center=<2 0 0>, rotation=<rot 90 0 0>}]
 pack [{
   shapeType=Capsule
   length=8.2429
   radius=10.7002
   shapeName=pack_capsule
   center=<-1.3358 -4.7527 -1.3147>
   rotation=<rot 0.3266 -86.5544 -179.5776>
  }]
}
constraints [
 {bone1Name=dyn_rope_01, bone2Name=base_rope, swing1LimitDegrees=10, swing2LimitDegrees=15, twistLimitDegrees=1}
 {bone1Name=dyn_rope_02, bone2Name=dyn_rope_01, swing1LimitDegrees=15, swing2LimitDegrees=25, twistLimitDegrees=1}
 {bone1Name=dyn_rope_03, bone2Name=dyn_rope_02, swing1LimitDegrees=50, swing2LimitDegrees=50, twistLimitDegrees=1}
 {
  bone1Name=base_rope
  bone2Name=pack
  swing1LimitDegrees=45
  swing1Motion=Free
  swing2LimitDegrees=45
  swing2Motion=Free
  twistLimitDegrees=45
  twistMotion=Free
 }
]
